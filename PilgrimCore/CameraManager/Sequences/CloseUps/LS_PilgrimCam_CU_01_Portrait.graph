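# Level Sequence for a Pilgrim camera-manager close-up portrait shot, cutting to a cine camera with a fixed, slightly rolled framing.
# The camera carries an attach track and a look-at offset, with constant lens, aperture and manual focus and an empty shake track.
# An Actor binding sits at identity transform.
# An animated, attached FortPlayerMannequin has its spawn track off, so it does not appear by default.
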LevelSequence /PilgrimCore/CameraManager/Sequences/CloseUps/LS_PilgrimCam_CU_01_Portrait
bindings [
 Actor1 {
  id=#1
  possessed_class=Actor
  tracks [
   {type=MovieSceneSpawnTrack, sections [{type=MovieSceneSpawnSection, channels [None {type=MovieSceneScriptingBoolChannel, default=true}]}]}
   {
    type=MovieScene3DTransformTrack
    sections [{
      type=MovieScene3DTransformSection
      channels [
       Location.X {type=MovieSceneScriptingDoubleChannel, default=0}
       Location.Y {type=MovieSceneScriptingDoubleChannel, default=0}
       Location.Z {type=MovieSceneScriptingDoubleChannel, default=0}
       Rotation.X {type=MovieSceneScriptingDoubleChannel, default=0}
       Rotation.Y {type=MovieSceneScriptingDoubleChannel, default=0}
       Rotation.Z {type=MovieSceneScriptingDoubleChannel, default=0}
       Scale.X {type=MovieSceneScriptingDoubleChannel, default=1}
       Scale.Y {type=MovieSceneScriptingDoubleChannel, default=1}
       Scale.Z {type=MovieSceneScriptingDoubleChannel, default=1}
      ]
     }]
   }
  ]
 }
 CameraComponent {
  id=#2
  parent_id=#3
  possessed_class=CineCameraComponent
  tracks [
   {
    type=MovieSceneFloatTrack
    display_name=Current Focal Length
    sections [{type=MovieSceneFloatSection, channels [None {type=MovieSceneScriptingFloatChannel, default=50}]}]
   }
   {
    type=MovieSceneFloatTrack
    display_name="Manual Focus Distance (Focus Settings)"
    sections [{type=MovieSceneFloatSection, channels [None {type=MovieSceneScriptingFloatChannel, default=589.9305, const [471.4661,1]}]}]
   }
   {
    type=MovieSceneFloatTrack
    display_name=Current Aperture
    sections [{type=MovieSceneFloatSection, channels [None {type=MovieSceneScriptingFloatChannel, default=2.8}]}]
   }
   {type=MovieSceneCameraShakeTrack, sections []}
   {
    type=MovieSceneBoolTrack
    display_name=Constrain Aspect Ratio
    sections [{type=MovieSceneBoolSection, channels [None {type=MovieSceneScriptingBoolChannel, default=false}]}]
   }
  ]
 }
 Fort Player Mannequin {
  id=#4
  possessed_class=FortPlayerMannequin
  tracks [
   {
    type=MovieScene3DTransformTrack
    sections [{
      type=MovieScene3DTransformSection
      channels [
       Location.X {type=MovieSceneScriptingDoubleChannel, default=0}
       Location.Y {type=MovieSceneScriptingDoubleChannel, default=0}
       Location.Z {type=MovieSceneScriptingDoubleChannel, default=-90.7379}
       Rotation.X {type=MovieSceneScriptingDoubleChannel, default=0}
       Rotation.Y {type=MovieSceneScriptingDoubleChannel, default=0}
       Rotation.Z {type=MovieSceneScriptingDoubleChannel, default=-90}
       Scale.X {type=MovieSceneScriptingDoubleChannel, default=1}
       Scale.Y {type=MovieSceneScriptingDoubleChannel, default=1}
       Scale.Z {type=MovieSceneScriptingDoubleChannel, default=1}
      ]
     }]
   }
   {type=MovieSceneSkeletalAnimationTrack, sections [{type=MovieSceneSkeletalAnimationSection, end_frame=354, start_frame=0}]}
   {type=MovieSceneSpawnTrack, sections [{type=MovieSceneSpawnSection, channels [None {type=MovieSceneScriptingBoolChannel, default=false}]}]}
   {type=MovieScene3DAttachTrack, sections [{type=MovieScene3DAttachSection, end_frame=354, start_frame=0}]}
  ]
 }
 CineCameraActor1 {
  id=#3
  possessed_class=CineCameraActor
  tracks [
   {
    type=MovieScene3DTransformTrack
    sections [{
      type=MovieScene3DTransformSection
      channels [
       Location.X {type=MovieSceneScriptingDoubleChannel, default=387, const [387,1]}
       Location.Y {type=MovieSceneScriptingDoubleChannel, default=-161, const [-161,1]}
       Location.Z {type=MovieSceneScriptingDoubleChannel, default=42, const [42,1]}
       Rotation.X {type=MovieSceneScriptingDoubleChannel, default=15, const [15,1]}
       Rotation.Y {type=MovieSceneScriptingDoubleChannel, default=-4, const [-4,1]}
       Rotation.Z {type=MovieSceneScriptingDoubleChannel, default=156, const [156,1]}
       Scale.X {type=MovieSceneScriptingDoubleChannel, default=1}
       Scale.Y {type=MovieSceneScriptingDoubleChannel, default=1}
       Scale.Z {type=MovieSceneScriptingDoubleChannel, default=1}
      ]
     }]
   }
   {type=MovieSceneSpawnTrack, sections [{type=MovieSceneSpawnSection, channels [None {type=MovieSceneScriptingBoolChannel, default=true}]}]}
   {type=MovieScene3DAttachTrack, sections [{type=MovieScene3DAttachSection, end_frame=348, start_frame=0}]}
   {
    type=MovieSceneDoubleVectorTrack
    display_name="Relative Offset (Lookat Tracking Settings)"
    sections [{
      type=MovieSceneDoubleVectorSection
      channels [
       Vector.X {type=MovieSceneScriptingDoubleChannel, default=-20}
       Vector.Y {type=MovieSceneScriptingDoubleChannel, default=30}
       Vector.Z {type=MovieSceneScriptingDoubleChannel, default=40}
      ]
     }]
   }
  ]
 }
]
display_rate {denominator=1, numerator=30}
tick_resolution {denominator=1, numerator=24000}
tracks [{type=MovieSceneCameraCutTrack, sections [{type=MovieSceneCameraCutSection, end_frame=300, start_frame=0}]}]
