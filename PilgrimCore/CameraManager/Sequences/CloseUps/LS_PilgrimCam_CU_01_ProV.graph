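# Level Sequence for a PilgrimCore close-up camera shot, a single camera cut of about ten seconds at 30 fps.
# A spawned CineCamera with an attach track holds a fixed transform and look-at offset while focal length gently zooms in, with constant manual focus and wide aperture.
# An empty spawned Actor sits at the origin, alongside an animated, attached soldier mannequin whose spawn track defaults to off, apparently a framing stand-in.
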LevelSequence /PilgrimCore/CameraManager/Sequences/CloseUps/LS_PilgrimCam_CU_01_ProV
bindings [
 Actor1 {
  id=#1
  possessed_class=Actor
  tracks [
   {type=MovieSceneSpawnTrack, sections [{type=MovieSceneSpawnSection, channels [None {type=MovieSceneScriptingBoolChannel, default=true}]}]}
   {
    type=MovieScene3DTransformTrack
    sections [{
      type=MovieScene3DTransformSection
      channels [
       Location.X {type=MovieSceneScriptingDoubleChannel, default=0}
       Location.Y {type=MovieSceneScriptingDoubleChannel, default=0}
       Location.Z {type=MovieSceneScriptingDoubleChannel, default=0}
       Rotation.X {type=MovieSceneScriptingDoubleChannel, default=0}
       Rotation.Y {type=MovieSceneScriptingDoubleChannel, default=0}
       Rotation.Z {type=MovieSceneScriptingDoubleChannel, default=0}
       Scale.X {type=MovieSceneScriptingDoubleChannel, default=1}
       Scale.Y {type=MovieSceneScriptingDoubleChannel, default=1}
       Scale.Z {type=MovieSceneScriptingDoubleChannel, default=1}
      ]
     }]
   }
  ]
 }
 CameraComponent {
  id=#2
  parent_id=#3
  possessed_class=CineCameraComponent
  tracks [
   {
    type=MovieSceneFloatTrack
    display_name=Current Focal Length
    sections [{type=MovieSceneFloatSection, channels [None {type=MovieSceneScriptingFloatChannel, default=100, keys [[0,90],[299,100]]}]}]
   }
   {
    type=MovieSceneFloatTrack
    display_name="Manual Focus Distance (Focus Settings)"
    sections [{type=MovieSceneFloatSection, channels [None {type=MovieSceneScriptingFloatChannel, default=589.9305, const [471.4661,1]}]}]
   }
   {
    type=MovieSceneFloatTrack
    display_name=Current Aperture
    sections [{type=MovieSceneFloatSection, channels [None {type=MovieSceneScriptingFloatChannel, default=2}]}]
   }
   {type=MovieSceneCameraShakeTrack, sections []}
  ]
 }
 SoldierMannequin_Male {
  id=#4
  possessed_class=SoldierMannequin_Male_C
  tracks [
   {
    type=MovieScene3DTransformTrack
    sections [{
      type=MovieScene3DTransformSection
      channels [
       Location.X {type=MovieSceneScriptingDoubleChannel, default=0}
       Location.Y {type=MovieSceneScriptingDoubleChannel, default=0}
       Location.Z {type=MovieSceneScriptingDoubleChannel, default=-138}
       Rotation.X {type=MovieSceneScriptingDoubleChannel, default=0}
       Rotation.Y {type=MovieSceneScriptingDoubleChannel, default=0}
       Rotation.Z {type=MovieSceneScriptingDoubleChannel, default=-90}
       Scale.X {type=MovieSceneScriptingDoubleChannel, default=1}
       Scale.Y {type=MovieSceneScriptingDoubleChannel, default=1}
       Scale.Z {type=MovieSceneScriptingDoubleChannel, default=1}
      ]
     }]
   }
   {type=MovieSceneSkeletalAnimationTrack, sections [{type=MovieSceneSkeletalAnimationSection, end_frame=300, start_frame=-5}]}
   {type=MovieSceneSpawnTrack, sections [{type=MovieSceneSpawnSection, channels [None {type=MovieSceneScriptingBoolChannel, default=false}]}]}
   {type=MovieScene3DAttachTrack, sections [{type=MovieScene3DAttachSection}]}
  ]
 }
 CineCameraActor1 {
  id=#3
  possessed_class=CineCameraActor
  tracks [
   {
    type=MovieScene3DTransformTrack
    sections [{
      type=MovieScene3DTransformSection
      channels [
       Location.X {type=MovieSceneScriptingDoubleChannel, default=575.274, const [449.9521,1]}
       Location.Y {type=MovieSceneScriptingDoubleChannel, default=-552.7704, const [-165.7859,1]}
       Location.Z {type=MovieSceneScriptingDoubleChannel, default=52.3141, const [5.4809,1]}
       Rotation.X {type=MovieSceneScriptingDoubleChannel, default=0}
       Rotation.Y {type=MovieSceneScriptingDoubleChannel, default=1, const [1,1]}
       Rotation.Z {type=MovieSceneScriptingDoubleChannel, default=159, const [159,1]}
       Scale.X {type=MovieSceneScriptingDoubleChannel, default=1}
       Scale.Y {type=MovieSceneScriptingDoubleChannel, default=1}
       Scale.Z {type=MovieSceneScriptingDoubleChannel, default=1}
      ]
     }]
   }
   {type=MovieSceneSpawnTrack, sections [{type=MovieSceneSpawnSection, channels [None {type=MovieSceneScriptingBoolChannel, default=true}]}]}
   {type=MovieScene3DAttachTrack, sections [{type=MovieScene3DAttachSection, end_frame=348, start_frame=-11}]}
   {
    type=MovieSceneDoubleVectorTrack
    display_name="Relative Offset (Lookat Tracking Settings)"
    sections [{
      type=MovieSceneDoubleVectorSection
      channels [
       Vector.X {type=MovieSceneScriptingDoubleChannel, default=0, const [0,1]}
       Vector.Y {type=MovieSceneScriptingDoubleChannel, default=0, const [0,1]}
       Vector.Z {type=MovieSceneScriptingDoubleChannel, default=20, const [20,1]}
      ]
     }]
   }
  ]
 }
]
display_rate {denominator=1, numerator=30}
tick_resolution {denominator=1, numerator=24000}
tracks [{type=MovieSceneCameraCutTrack, sections [{type=MovieSceneCameraCutSection, end_frame=300, start_frame=0}]}]
