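# Level Sequence for close-up shot 08 in the PilgrimCore camera manager, cutting to a spawned, attached CineCamera.
# The camera actor and its component carry constant transforms, focal length, manual focus and aperture, giving a fixed framing, plus a look-at tracking offset.
# An animated, attached Fort player mannequin has spawning off by default, likely as a framing stand-in.
# A spawned spotLight_relocator actor sits at identity transform.
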LevelSequence /PilgrimCore/CameraManager/Sequences/CloseUps/LS_PilgrimCam_CU_08
bindings [
 Fort Player Mannequin {
  id=#1
  possessed_class=FortPlayerMannequin
  tracks [
   {
    type=MovieScene3DTransformTrack
    sections [{
      type=MovieScene3DTransformSection
      channels [
       Location.X {type=MovieSceneScriptingDoubleChannel, default=0}
       Location.Y {type=MovieSceneScriptingDoubleChannel, default=0}
       Location.Z {type=MovieSceneScriptingDoubleChannel, default=-90.7379}
       Rotation.X {type=MovieSceneScriptingDoubleChannel, default=0}
       Rotation.Y {type=MovieSceneScriptingDoubleChannel, default=0}
       Rotation.Z {type=MovieSceneScriptingDoubleChannel, default=-90}
       Scale.X {type=MovieSceneScriptingDoubleChannel, default=1}
       Scale.Y {type=MovieSceneScriptingDoubleChannel, default=1}
       Scale.Z {type=MovieSceneScriptingDoubleChannel, default=1}
      ]
     }]
   }
   {type=MovieSceneSkeletalAnimationTrack, sections [{type=MovieSceneSkeletalAnimationSection, end_frame=354, start_frame=0}]}
   {type=MovieSceneSpawnTrack, sections [{type=MovieSceneSpawnSection, channels [None {type=MovieSceneScriptingBoolChannel, default=false}]}]}
   {type=MovieScene3DAttachTrack, sections [{type=MovieScene3DAttachSection, end_frame=354, start_frame=0}]}
  ]
 }
 spotLight_relocator {
  id=#2
  possessed_class=Actor
  tracks [
   {type=MovieSceneSpawnTrack, sections [{type=MovieSceneSpawnSection, channels [None {type=MovieSceneScriptingBoolChannel, default=true}]}]}
   {
    type=MovieScene3DTransformTrack
    sections [{
      type=MovieScene3DTransformSection
      channels [
       Location.X {type=MovieSceneScriptingDoubleChannel, default=0}
       Location.Y {type=MovieSceneScriptingDoubleChannel, default=0}
       Location.Z {type=MovieSceneScriptingDoubleChannel, default=0}
       Rotation.X {type=MovieSceneScriptingDoubleChannel, default=0}
       Rotation.Y {type=MovieSceneScriptingDoubleChannel, default=0}
       Rotation.Z {type=MovieSceneScriptingDoubleChannel, default=0}
       Scale.X {type=MovieSceneScriptingDoubleChannel, default=1}
       Scale.Y {type=MovieSceneScriptingDoubleChannel, default=1}
       Scale.Z {type=MovieSceneScriptingDoubleChannel, default=1}
      ]
     }]
   }
  ]
 }
 CameraComponent {
  id=#3
  parent_id=#4
  possessed_class=CineCameraComponent
  tracks [
   {
    type=MovieSceneFloatTrack
    display_name=Current Focal Length
    sections [{type=MovieSceneFloatSection, channels [None {type=MovieSceneScriptingFloatChannel, default=22.0693, const [17.4784,1]}]}]
   }
   {
    type=MovieSceneFloatTrack
    display_name="Manual Focus Distance (Focus Settings)"
    sections [{type=MovieSceneFloatSection, channels [None {type=MovieSceneScriptingFloatChannel, default=150.8008, const [150.8008,1]}]}]
   }
   {
    type=MovieSceneFloatTrack
    display_name=Current Aperture
    sections [{type=MovieSceneFloatSection, channels [None {type=MovieSceneScriptingFloatChannel, default=1}]}]
   }
   {
    type=MovieScene3DTransformTrack
    sections [{
      type=MovieScene3DTransformSection
      channels [
       Location.X {type=MovieSceneScriptingDoubleChannel, default=114.8961, const [137.2159,1]}
       Location.Y {type=MovieSceneScriptingDoubleChannel, default=103.3204, const [90.5986,1]}
       Location.Z {type=MovieSceneScriptingDoubleChannel, default=-22.8639, const [-23.0554,1]}
       Rotation.X {type=MovieSceneScriptingDoubleChannel, default=9.7594, const [47.6393,1]}
       Rotation.Y {type=MovieSceneScriptingDoubleChannel, default=10.3705, const [11.9034,1]}
       Rotation.Z {type=MovieSceneScriptingDoubleChannel, default=-27.8367, const [-25.2211,1]}
       Scale.X {type=MovieSceneScriptingDoubleChannel, default=1, const [1,1]}
       Scale.Y {type=MovieSceneScriptingDoubleChannel, default=1, const [1,1]}
       Scale.Z {type=MovieSceneScriptingDoubleChannel, default=1, const [1,1]}
      ]
     }]
   }
  ]
 }
 Cine Camera Actor {
  id=#4
  possessed_class=CineCameraActor
  tracks [
   {
    type=MovieScene3DTransformTrack
    sections [{
      type=MovieScene3DTransformSection
      channels [
       Location.X {type=MovieSceneScriptingDoubleChannel, default=-5150.5862, const [288.0978,1]}
       Location.Y {type=MovieSceneScriptingDoubleChannel, default=-552.7704, const [9.8026,1]}
       Location.Z {type=MovieSceneScriptingDoubleChannel, default=671.3617, const [-35.4635,1]}
       Rotation.X {type=MovieSceneScriptingDoubleChannel, default=-16, const [-1.4199,1]}
       Rotation.Y {type=MovieSceneScriptingDoubleChannel, default=16, const [16,1]}
       Rotation.Z {type=MovieSceneScriptingDoubleChannel, default=172, const [172,1]}
       Scale.X {type=MovieSceneScriptingDoubleChannel, default=1, const [1,1]}
       Scale.Y {type=MovieSceneScriptingDoubleChannel, default=1, const [1,1]}
       Scale.Z {type=MovieSceneScriptingDoubleChannel, default=1, const [1,1]}
      ]
     }]
   }
   {type=MovieSceneSpawnTrack, sections [{type=MovieSceneSpawnSection, channels [None {type=MovieSceneScriptingBoolChannel, default=true}]}]}
   {type=MovieScene3DAttachTrack, sections [{type=MovieScene3DAttachSection, end_frame=348, start_frame=0}]}
   {
    type=MovieSceneDoubleVectorTrack
    display_name="Relative Offset (Lookat Tracking Settings)"
    sections [{
      type=MovieSceneDoubleVectorSection
      channels [
       Vector.X {type=MovieSceneScriptingDoubleChannel, default=-10}
       Vector.Y {type=MovieSceneScriptingDoubleChannel, default=20}
       Vector.Z {type=MovieSceneScriptingDoubleChannel, default=80}
      ]
     }]
   }
  ]
 }
]
display_rate {denominator=1, numerator=30}
tick_resolution {denominator=1, numerator=24000}
tracks [{type=MovieSceneCameraCutTrack, sections [{type=MovieSceneCameraCutSection, end_frame=300, start_frame=0}]}]
