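# Close-up camera shot sequence (CU_10) for the Pilgrim camera manager, with a cine camera cut over frames 0-300.
# The attached camera pushes in and tilts over the first 72 frames while focal length widens and focus pulls closer, then focus eases back slightly.
# An attached HeadLookAtRef marker shares its keys with the camera's look-at offset.
# An animated mannequin defaults to unspawned, and a spotLight_relocator actor stays at zero transform.
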LevelSequence /PilgrimCore/CameraManager/Sequences/CloseUps/LS_PilgrimCam_CU_10
bindings [
 Fort Player Mannequin {
  id=#1
  possessed_class=FortPlayerMannequin
  tracks [
   {
    type=MovieScene3DTransformTrack
    sections [{
      type=MovieScene3DTransformSection
      channels [
       Location.X {type=MovieSceneScriptingDoubleChannel, default=0}
       Location.Y {type=MovieSceneScriptingDoubleChannel, default=0}
       Location.Z {type=MovieSceneScriptingDoubleChannel, default=-90.7379}
       Rotation.X {type=MovieSceneScriptingDoubleChannel, default=0}
       Rotation.Y {type=MovieSceneScriptingDoubleChannel, default=0}
       Rotation.Z {type=MovieSceneScriptingDoubleChannel, default=-90}
       Scale.X {type=MovieSceneScriptingDoubleChannel, default=1}
       Scale.Y {type=MovieSceneScriptingDoubleChannel, default=1}
       Scale.Z {type=MovieSceneScriptingDoubleChannel, default=1}
      ]
     }]
   }
   {type=MovieSceneSpawnTrack, sections [{type=MovieSceneSpawnSection, channels [None {type=MovieSceneScriptingBoolChannel, default=false}]}]}
   {type=MovieScene3DAttachTrack, sections [{type=MovieScene3DAttachSection, end_frame=354, start_frame=0}]}
   {type=MovieSceneSkeletalAnimationTrack, sections [{type=MovieSceneSkeletalAnimationSection, end_frame=340, start_frame=0}]}
  ]
 }
 spotLight_relocator {
  id=#2
  possessed_class=Actor
  tracks [
   {type=MovieSceneSpawnTrack, sections [{type=MovieSceneSpawnSection, channels [None {type=MovieSceneScriptingBoolChannel, default=true}]}]}
   {
    type=MovieScene3DTransformTrack
    sections [{
      type=MovieScene3DTransformSection
      channels [
       Location.X {type=MovieSceneScriptingDoubleChannel, default=0}
       Location.Y {type=MovieSceneScriptingDoubleChannel, default=0}
       Location.Z {type=MovieSceneScriptingDoubleChannel, default=0}
       Rotation.X {type=MovieSceneScriptingDoubleChannel, default=0}
       Rotation.Y {type=MovieSceneScriptingDoubleChannel, default=0}
       Rotation.Z {type=MovieSceneScriptingDoubleChannel, default=0}
       Scale.X {type=MovieSceneScriptingDoubleChannel, default=1}
       Scale.Y {type=MovieSceneScriptingDoubleChannel, default=1}
       Scale.Z {type=MovieSceneScriptingDoubleChannel, default=1}
      ]
     }]
   }
  ]
 }
 HeadLookAtRef {
  id=#3
  possessed_class=FortStaticMeshActor
  tracks [
   {
    type=MovieScene3DTransformTrack
    sections [{
      type=MovieScene3DTransformSection
      channels [
       Location.X {type=MovieSceneScriptingDoubleChannel, default=-18.292, keys [[0,-18.292],[72,-16.673]]}
       Location.Y {type=MovieSceneScriptingDoubleChannel, default=9.7756, keys [[0,9.7756],[72,16.1362]]}
       Location.Z {type=MovieSceneScriptingDoubleChannel, default=72.2296, keys [[0,72.2296],[72,37.7569]]}
       Rotation.X {type=MovieSceneScriptingDoubleChannel, default=0, const [0,2]}
       Rotation.Y {type=MovieSceneScriptingDoubleChannel, default=0, const [0,2]}
       Rotation.Z {type=MovieSceneScriptingDoubleChannel, default=0, const [0,2]}
       Scale.X {type=MovieSceneScriptingDoubleChannel, default=0.1, const [0.1,2]}
       Scale.Y {type=MovieSceneScriptingDoubleChannel, default=0.1, const [0.1,2]}
       Scale.Z {type=MovieSceneScriptingDoubleChannel, default=0.1, const [0.1,2]}
      ]
     }]
   }
   {type=MovieSceneSpawnTrack, sections [{type=MovieSceneSpawnSection, channels [None {type=MovieSceneScriptingBoolChannel, default=true}]}]}
   {type=MovieScene3DAttachTrack, sections [{type=MovieScene3DAttachSection, end_frame=354, start_frame=0}]}
  ]
 }
 CameraComponent {
  id=#4
  parent_id=#5
  possessed_class=CineCameraComponent
  tracks [
   {
    type=MovieSceneFloatTrack
    display_name=Current Focal Length
    sections [{type=MovieSceneFloatSection, channels [None {type=MovieSceneScriptingFloatChannel, default=17.088, keys [[0.0013,24.0489],[72,17.088]]}]}]
   }
   {
    type=MovieSceneFloatTrack
    display_name="Manual Focus Distance (Focus Settings)"
    sections [{
      type=MovieSceneFloatSection
      channels [None {type=MovieSceneScriptingFloatChannel, default=128.914, keys [[0.0013,265.902],[72,128.914],[220,146.732]]}]
     }]
   }
   {
    type=MovieSceneFloatTrack
    display_name=Current Aperture
    sections [{type=MovieSceneFloatSection, channels [None {type=MovieSceneScriptingFloatChannel, default=1}]}]
   }
   {type=MovieSceneCameraShakeTrack, sections [{type=MovieSceneCameraShakeSection, end_frame=648, start_frame=348}]}
  ]
 }
 Cine Camera Actor {
  id=#5
  possessed_class=CineCameraActor
  tracks [
   {
    type=MovieScene3DTransformTrack
    sections [{
      type=MovieScene3DTransformSection
      channels [
       Location.X {type=MovieSceneScriptingDoubleChannel, default=-5150.5862, keys [[0.0013,288.0978],[72,142.591]]}
       Location.Y {type=MovieSceneScriptingDoubleChannel, default=-552.7704, keys [[0,33.5043],[72,2.6379]]}
       Location.Z {type=MovieSceneScriptingDoubleChannel, default=671.3617, keys [[0.0013,-35.4635],[72,-14.4895]]}
       Rotation.X {type=MovieSceneScriptingDoubleChannel, default=0, keys [[0,-1.2799],[0.0013,-1.4199],[72,-1.4506]]}
       Rotation.Y {type=MovieSceneScriptingDoubleChannel, default=-13.2838, keys [[0,1.94],[0.0013,8.0759],[72,14.2658]]}
       Rotation.Z {type=MovieSceneScriptingDoubleChannel, default=179.1, keys [[0,169.1815],[0.0013,169.8598],[72,169.7018]]}
       Scale.X {type=MovieSceneScriptingDoubleChannel, default=1, const [1,3]}
       Scale.Y {type=MovieSceneScriptingDoubleChannel, default=1, const [1,3]}
       Scale.Z {type=MovieSceneScriptingDoubleChannel, default=1, const [1,3]}
      ]
     }]
   }
   {type=MovieSceneSpawnTrack, sections [{type=MovieSceneSpawnSection, channels [None {type=MovieSceneScriptingBoolChannel, default=true}]}]}
   {type=MovieScene3DAttachTrack, sections [{type=MovieScene3DAttachSection, end_frame=348, start_frame=0}]}
   {
    type=MovieSceneDoubleVectorTrack
    display_name="Relative Offset (Lookat Tracking Settings)"
    sections [{
      type=MovieSceneDoubleVectorSection
      channels [
       Vector.X {type=MovieSceneScriptingDoubleChannel, default=-53.2312, thinned_from=3, keys [[0,-18.292],[72,-16.673]]}
       Vector.Y {type=MovieSceneScriptingDoubleChannel, default=0, thinned_from=3, keys [[0,9.7756],[72,16.1362]]}
       Vector.Z {type=MovieSceneScriptingDoubleChannel, default=-28.2535, thinned_from=3, keys [[0,72.2296],[72,37.7569]]}
      ]
     }]
   }
  ]
 }
]
display_rate {denominator=1, numerator=30}
tick_resolution {denominator=1, numerator=24000}
tracks [{type=MovieSceneCameraCutTrack, sections [{type=MovieSceneCameraCutSection, end_frame=300, start_frame=0}]}]
