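# Close-up camera Level Sequence 18 for the PilgrimCore camera manager: a 300-frame shot at 30 fps with one camera cut.
# An attached, spawned CineCamera sweeps a keyed arc at constant height, yawing about 170 degrees, with fixed wide focal length, very wide aperture and manual focus.
# An animated soldier mannequin, apparently a framing stand-in, has its spawn track off, while a static spotLight_relocator actor spawns at zero transform.
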LevelSequence /PilgrimCore/CameraManager/Sequences/CloseUps/LS_PilgrimCam_CU_18_ProV
bindings [
 CameraComponent {
  id=#1
  parent_id=#2
  possessed_class=CineCameraComponent
  tracks [
   {
    type=MovieSceneFloatTrack
    display_name=Current Focal Length
    sections [{type=MovieSceneFloatSection, channels [None {type=MovieSceneScriptingFloatChannel, default=13}]}]
   }
   {
    type=MovieSceneFloatTrack
    display_name="Manual Focus Distance (Focus Settings)"
    sections [{type=MovieSceneFloatSection, channels [None {type=MovieSceneScriptingFloatChannel, default=182.8572}]}]
   }
   {
    type=MovieSceneFloatTrack
    display_name=Current Aperture
    sections [{type=MovieSceneFloatSection, channels [None {type=MovieSceneScriptingFloatChannel, default=0.6}]}]
   }
   {
    type=MovieScene3DTransformTrack
    sections [{
      type=MovieScene3DTransformSection
      channels [
       Location.X {
        type=MovieSceneScriptingDoubleChannel
        default=63.109
        keys [[0,338.9972],[39,314.0604],[73,261.7473],[136,137.6089],[173,70.5158],[207,29.6921],[254,-1.6184],[299,-7.9219]]
       }
       Location.Y {
        type=MovieSceneScriptingDoubleChannel
        default=-88.6678
        keys [[0,-107.7145],[39,-186.9693],[73,-233.7579],[136,-257.1423],[173,-230.007],[254,-99.2647],[299,-34.3501]]
       }
       Location.Z {type=MovieSceneScriptingDoubleChannel, default=74, const [74,1]}
       Rotation.X {type=MovieSceneScriptingDoubleChannel, default=18.9379, const [-0.0753,1]}
       Rotation.Y {type=MovieSceneScriptingDoubleChannel, default=-4, const [-4,1]}
       Rotation.Z {
        type=MovieSceneScriptingDoubleChannel
        default=40
        thinned_from=6
        keys [[0,-223.2956],[73,-266.8637],[207,-344.1431],[254,-368.001],[299,-391.7121]]
       }
       Scale.X {type=MovieSceneScriptingDoubleChannel, default=1, const [1,5]}
       Scale.Y {type=MovieSceneScriptingDoubleChannel, default=1, const [1,5]}
       Scale.Z {type=MovieSceneScriptingDoubleChannel, default=1, const [1,5]}
      ]
     }]
   }
  ]
 }
 SoldierMannequin_Male {
  id=#3
  possessed_class=SoldierMannequin_Male_C
  tracks [
   {
    type=MovieScene3DTransformTrack
    sections [{
      type=MovieScene3DTransformSection
      channels [
       Location.X {type=MovieSceneScriptingDoubleChannel, default=0}
       Location.Y {type=MovieSceneScriptingDoubleChannel, default=0}
       Location.Z {type=MovieSceneScriptingDoubleChannel, default=-80}
       Rotation.X {type=MovieSceneScriptingDoubleChannel, default=0}
       Rotation.Y {type=MovieSceneScriptingDoubleChannel, default=0}
       Rotation.Z {type=MovieSceneScriptingDoubleChannel, default=-90}
       Scale.X {type=MovieSceneScriptingDoubleChannel, default=1}
       Scale.Y {type=MovieSceneScriptingDoubleChannel, default=1}
       Scale.Z {type=MovieSceneScriptingDoubleChannel, default=1}
      ]
     }]
   }
   {type=MovieSceneSkeletalAnimationTrack, sections [{type=MovieSceneSkeletalAnimationSection, end_frame=300, start_frame=0}]}
   {type=MovieSceneSpawnTrack, sections [{type=MovieSceneSpawnSection, channels [None {type=MovieSceneScriptingBoolChannel, default=false}]}]}
  ]
 }
 spotLight_relocator {
  id=#4
  possessed_class=Actor
  tracks [
   {type=MovieSceneSpawnTrack, sections [{type=MovieSceneSpawnSection, channels [None {type=MovieSceneScriptingBoolChannel, default=true}]}]}
   {
    type=MovieScene3DTransformTrack
    sections [{
      type=MovieScene3DTransformSection
      channels [
       Location.X {type=MovieSceneScriptingDoubleChannel, default=0}
       Location.Y {type=MovieSceneScriptingDoubleChannel, default=0}
       Location.Z {type=MovieSceneScriptingDoubleChannel, default=0}
       Rotation.X {type=MovieSceneScriptingDoubleChannel, default=0}
       Rotation.Y {type=MovieSceneScriptingDoubleChannel, default=0}
       Rotation.Z {type=MovieSceneScriptingDoubleChannel, default=0}
       Scale.X {type=MovieSceneScriptingDoubleChannel, default=1}
       Scale.Y {type=MovieSceneScriptingDoubleChannel, default=1}
       Scale.Z {type=MovieSceneScriptingDoubleChannel, default=1}
      ]
     }]
   }
  ]
 }
 Cine Camera Actor {
  id=#2
  possessed_class=CineCameraActor
  tracks [
   {type=MovieSceneSpawnTrack, sections [{type=MovieSceneSpawnSection, channels [None {type=MovieSceneScriptingBoolChannel, default=true}]}]}
   {type=MovieSceneCameraShakeTrack, sections []}
   {type=MovieScene3DAttachTrack, sections [{type=MovieScene3DAttachSection, end_frame=300, start_frame=0}]}
  ]
 }
]
display_rate {denominator=1, numerator=30}
tick_resolution {denominator=1, numerator=24000}
tracks [{type=MovieSceneCameraCutTrack, sections [{type=MovieSceneCameraCutSection, end_frame=300, start_frame=0}]}]
