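# Level Sequence for a PilgrimCore camera-manager close-up shot, cutting to a spawned, attached cine camera.
# The camera component sweeps around in a keyed arc over roughly the first half of the cut, then holds, with a camera shake section throughout.
# Fixed wide focal length pairs with animated manual focus distance and aperture.
# An animated, attached soldier mannequin has spawning off, likely as a framing stand-in, alongside a spawned spotlight relocator actor.
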LevelSequence /PilgrimCore/CameraManager/Sequences/CloseUps/LS_PilgrimCam_CU_22_ProV
bindings [
 spotLight_relocator {
  id=#1
  possessed_class=Actor
  tracks [
   {type=MovieSceneSpawnTrack, sections [{type=MovieSceneSpawnSection, channels [None {type=MovieSceneScriptingBoolChannel, default=true}]}]}
   {
    type=MovieScene3DTransformTrack
    sections [{
      type=MovieScene3DTransformSection
      channels [
       Location.X {type=MovieSceneScriptingDoubleChannel, default=0}
       Location.Y {type=MovieSceneScriptingDoubleChannel, default=0}
       Location.Z {type=MovieSceneScriptingDoubleChannel, default=0}
       Rotation.X {type=MovieSceneScriptingDoubleChannel, default=0}
       Rotation.Y {type=MovieSceneScriptingDoubleChannel, default=0}
       Rotation.Z {type=MovieSceneScriptingDoubleChannel, default=0}
       Scale.X {type=MovieSceneScriptingDoubleChannel, default=1}
       Scale.Y {type=MovieSceneScriptingDoubleChannel, default=1}
       Scale.Z {type=MovieSceneScriptingDoubleChannel, default=1}
      ]
     }]
   }
  ]
 }
 SoldierMannequin_Male {
  id=#2
  possessed_class=SoldierMannequin_Male_C
  tracks [
   {
    type=MovieScene3DTransformTrack
    sections [{
      type=MovieScene3DTransformSection
      channels [
       Location.X {type=MovieSceneScriptingDoubleChannel, default=0}
       Location.Y {type=MovieSceneScriptingDoubleChannel, default=0}
       Location.Z {type=MovieSceneScriptingDoubleChannel, default=-90}
       Rotation.X {type=MovieSceneScriptingDoubleChannel, default=0}
       Rotation.Y {type=MovieSceneScriptingDoubleChannel, default=0}
       Rotation.Z {type=MovieSceneScriptingDoubleChannel, default=-90}
       Scale.X {type=MovieSceneScriptingDoubleChannel, default=1}
       Scale.Y {type=MovieSceneScriptingDoubleChannel, default=1}
       Scale.Z {type=MovieSceneScriptingDoubleChannel, default=1}
      ]
     }]
   }
   {type=MovieSceneSkeletalAnimationTrack, sections [{type=MovieSceneSkeletalAnimationSection, end_frame=303, start_frame=-13}]}
   {type=MovieSceneSpawnTrack, sections [{type=MovieSceneSpawnSection, channels [None {type=MovieSceneScriptingBoolChannel, default=false}]}]}
   {type=MovieScene3DAttachTrack, sections [{type=MovieScene3DAttachSection}]}
  ]
 }
 CameraComponent {
  id=#3
  parent_id=#4
  possessed_class=CineCameraComponent
  tracks [
   {
    type=MovieSceneFloatTrack
    display_name=Current Focal Length
    sections [{type=MovieSceneFloatSection, channels [None {type=MovieSceneScriptingFloatChannel, default=20}]}]
   }
   {
    type=MovieSceneFloatTrack
    display_name="Manual Focus Distance (Focus Settings)"
    sections [{
      type=MovieSceneFloatSection
      channels [None {type=MovieSceneScriptingFloatChannel, default=121.7907, keys [[0,378.0359],[39,121.7907],[103.74,184.1251]]}]
     }]
   }
   {
    type=MovieSceneFloatTrack
    display_name=Current Aperture
    sections [{type=MovieSceneFloatSection, channels [None {type=MovieSceneScriptingFloatChannel, default=1.2, keys [[0,0.5],[103.74,1.2]]}]}]
   }
   {type=MovieSceneCameraShakeTrack, sections [{type=MovieSceneCameraShakeSection, end_frame=300}]}
   {
    type=MovieScene3DTransformTrack
    sections [{
      type=MovieScene3DTransformSection
      channels [
       Location.X {
        type=MovieSceneScriptingDoubleChannel
        default=-4989.0628
        keys [[0,-6486.9028],[9.1413,-6514.4219],[32.1012,-6644.2197],[58.5063,-6872.3755],[102.0063,-6918.5283],[152,-6919.0249]]
       }
       Location.Y {
        type=MovieSceneScriptingDoubleChannel
        default=-715.5293
        keys [[0,476.1655],[9.1413,493.9617],[32.1012,609.1184],[58.5063,648.5812],[102.0063,594.1679],[152.7437,587.6233]]
       }
       Location.Z {type=MovieSceneScriptingDoubleChannel, default=-331, keys [[0,-325.2011],[9.1413,-335.4652],[32.1012,-331],[58.5063,-365],[152.7437,-366]]}
       Rotation.X {
        type=MovieSceneScriptingDoubleChannel
        default=46.5491
        keys [[0,2.3133],[9.1413,2.3273],[32.1012,0.4027],[58.5063,-2.8681],[102.0063,3.0921],[152.7437,1.0898]]
       }
       Rotation.Y {
        type=MovieSceneScriptingDoubleChannel
        default=7
        keys [[0,-7.1717],[9.1413,-7.1424],[32.1012,-9.6547],[58.5063,0.6455],[102,7],[152.7437,6]]
       }
       Rotation.Z {
        type=MovieSceneScriptingDoubleChannel
        default=323
        keys [[0,164.8746],[9.1413,168.4273],[32.1012,195.413],[58.5063,296.8233],[102.0063,322],[152.7437,323]]
       }
       Scale.X {type=MovieSceneScriptingDoubleChannel, default=1, const [1,6]}
       Scale.Y {type=MovieSceneScriptingDoubleChannel, default=1, const [1,6]}
       Scale.Z {type=MovieSceneScriptingDoubleChannel, default=1, const [1,6]}
      ]
     }]
   }
  ]
 }
 CineCameraActor1 {
  id=#4
  possessed_class=CineCameraActor
  tracks [
   {type=MovieSceneSpawnTrack, sections [{type=MovieSceneSpawnSection, channels [None {type=MovieSceneScriptingBoolChannel, default=true}]}]}
   {type=MovieScene3DAttachTrack, sections [{type=MovieScene3DAttachSection, end_frame=347, start_frame=0}]}
  ]
 }
]
display_rate {denominator=1, numerator=30}
tick_resolution {denominator=1, numerator=24000}
tracks [{type=MovieSceneCameraCutTrack, sections [{type=MovieSceneCameraCutSection, end_frame=300, start_frame=0}]}]
